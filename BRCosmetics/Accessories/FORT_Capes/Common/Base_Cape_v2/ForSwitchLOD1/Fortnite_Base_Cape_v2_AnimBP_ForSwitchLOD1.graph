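# Animation Blueprint for the base cape v2 cosmetic (Switch LOD1 variant) that drives the Base_Cape_01 Chaos cloth.
# Each update runs the nativized cloth controller with a parameter asset, core and wind gust settings, and caches the cloth interactor.
# It resets and teleports the cloth shortly after zipline entry, and on the fall end following zipline exit.
# AnimGraph has no pose input; OnInitialUpdate overrides a rigid body node's physics asset, and debug draws are unwired.
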AnimBlueprint /BRCosmetics/Accessories/FORT_Capes/Common/Base_Cape_v2/ForSwitchLOD1/Fortnite_Base_Cape_v2_AnimBP_ForSwitchLOD1
target_skeleton=/Game/Accessories/FORT_Capes/Common/Base_Cape_v2/Fortnite_Base_Cape_v2_Skeleton
graphs [
 AnimGraph {nodes [Output Pose() {layerGroup=None}]}
 EventGraph {nodes [
   n1 = Event BlueprintUpdateAnimation()
   n2 = "Cloth Controller Function (Parameter Asset) (Nativized)"!(
    execute=@n10.then_0
    CharacterPartAnimInstance:Custom Character Part Anim Instance Object Reference=Self-Reference()
    ClothingAssetName:Name=Base_Cape_01
    ClothCoreSettings:"Cloth Core Settings Structure (by ref)"=Get Cloth Core Settings_0() -> (Cloth Core Settings_0:Cloth Core Settings Structure)
    WindGustLocalSettings:"Wind Gust Settings Structure (by ref)"=Get Wind Gust Local Settings() -> (Wind Gust Local Settings:Wind Gust Settings Structure)
    ClothUpAxis:Vector="0,0,1"
    LinVelScaleFwdBwdClamp:float=0
    LinVelScaleUpDwnClamp:float=0.7
    DeltaTime:float=@n1.DeltaTimeX
    GameTimeInSeconds:float=GetGameTimeInSeconds()
    StateLogicLayerName:Name=none
    Parameters:Fort Character Dynamics Parameters Object Reference=Get Cloth Controller Parameters() -> (Cloth Controller Parameters:Fort Character Dynamics Parameters Object Reference)
   ) -> (
    then
    OutCurrentState:Name=None
    OutClothParameters:Cloth Parameters Structure
    OutClothCoreSettings:Cloth Core Settings Structure
    OutWindGustSettings:Wind Gust Parameters Structure
    OutWindGustLocalSettings:Wind Gust Settings Structure
    OutWindVelocity:Vector
    OutLinearVelVector:Vector
   )
   n3 = GetOwnerSkeletalMeshComponent()
   DrawVectorFromJointFunction!(
    execute
    SkelMeshComponent:Scene Component Object Reference=@n3
    JointName:Name=pelvis
    LineEnd:Vector=Break Cloth Parameters!(ClothParameters:"Cloth Parameters Structure (by ref)"=@n2.OutClothParameters) -> (
     LinearVelocityScale:Vector=~
     AngularVelocityScale:float=~
     FictitiousAngularScale:float=~
     GravityOverride:Vector=~
     DragCoefficient:Vector 2D Structure
     LiftCoefficient:Vector 2D Structure
     OuterDrag:Vector 2D Structure
     OuterLift:Vector 2D Structure
     MaxLinearVelocity:Vector=~
     MaxLinearAcceleration:Vector=~
     MaxAngularVelocity:float=~
     MaxAngularAcceleration:float=~
     bEnableLinearVelocityClamping:bool=~
     bEnableLinearAccelerationClamping:bool=~
     bEnableAngularVelocityClamping:bool=~
     bEnableAngularAccelerationClamping:bool=~
     JointName:Name
    ) {is=title K2Node_BreakStruct}.GravityOverride
    JointOffset:Vector="0, 0, 0"
    LineColor:Linear Color Structure=(R=1,G=0,B=0,A=1)
    Thickness:double=0.8
    Foreground:EDrawDebugSceneDepthPriorityGroup Enum=Foreground
   ) -> (then)
   > DrawJointDebugFunction!(
    execute
    SkelMeshComponent:Scene Component Object Reference=@n3
    JointOffset:Vector="0, 0, 0"
    JointName:Array of Names=Make Array("[0]":Name=pelvis) -> (Array:Array of Names)
    Scale:double=7
    Duration:double=0
    Thickness:double=0.7
    DebugDraw:bool=true
    Foreground:EDrawDebugSceneDepthPriorityGroup Enum=Foreground
   ) -> (then)
   n4 = Event BlueprintProcessAnimInputEvent()
   > n5 = Branch(Condition=IsNamed(self=@n4.AnimInputEvent, InEventName=aie.Player.Zipline.Entered))
   > Delay(Duration=0.1)
   n6 = Get ChaosClothingInteractor() -> (ChaosClothingInteractor:Chaos Clothing Interactor Object Reference)
   > ResetAndTeleport!(execute, self:Chaos Clothing Interactor Object Reference=@n6, bReset:bool=true, bTeleport:bool=true) -> (then)
   Branch(execute=@n5.else, Condition=IsNamed(self=@n4.AnimInputEvent, InEventName=aie.Player.Zipline.Exited))
   > Set resetClothLandingFromZilpine(resetClothLandingFromZilpine:bool=true) -> (Output_Get:bool)
   Event OnEndFalling()
   > Branch(Condition=Get resetClothLandingFromZilpine() -> (resetClothLandingFromZilpine:bool))
   > ResetAndTeleport(self=@n6, bReset=true, bTeleport=true)
   > Set resetClothLandingFromZilpine(resetClothLandingFromZilpine=false)
   DrawVectorFromJointFunction(
    SkelMeshComponent=@n3
    JointName=pelvis
    LineEnd="0, 0, 0"
    JointOffset="0, 0, 0"
    LineColor=(R=0.9957,G=1,B=0,A=1)
    Thickness=2
    Foreground=Foreground
   )
   Set Wind Gust Local Settings Out(execute=@n2.then, Wind Gust Local Settings Out:Wind Gust Settings Structure=@n2.OutWindGustLocalSettings) -> (Output_Get:Wind Gust Settings Structure)
   n7 = GetClothingSimulationInteractor!(self:Skeletal Mesh Component Object Reference=GetOwningComponent(), ClothingAssetName:Name=None) -> (ReturnValue:Clothing Simulation Interactor Object Reference)
   n8 = MakeLiteralString(Value=Base_Cape_01)
   n9 = Cast To ChaosClothingInteractor!(
    execute=@n11.Is Valid
    Object=GetClothingInteractor!(self:Clothing Simulation Interactor Object Reference=@n7, ClothingAssetName:Name=String To Name(InString=@n8)) -> (ReturnValue:Clothing Interactor Object Reference)
   ) -> (then, CastFailed, AsChaos Clothing Interactor:Chaos Clothing Interactor Object Reference) {is=title K2Node_DynamicCast}
   > Set ChaosClothingInteractor(ChaosClothingInteractor:Chaos Clothing Interactor Object Reference=@n9) -> (Output_Get:Chaos Clothing Interactor Object Reference)
   n10 = Sequence!(execute=@n1.then) -> (then_0, then_1, then_2:Exec, then_3:Exec)
   n11 = Is Valid(exec=@n10.then_1, InputObject=@n7)
  ]}
 OnInitialUpdate {nodes [
   n1 = OnInitialUpdate!() -> (then, Context:"Anim Update Context Structure (by ref)", Node:"Anim Node Reference Structure (by ref)") {is=title K2Node_FunctionEntry}
   > n2 = ConvertToRigidBodyAnimNode!(execute, Node:"Anim Node Reference Structure (by ref)"=@n1.Node) -> (Succeeded:Exec, Failed:Exec, ReturnValue:Rigid Body Anim Node Reference Structure)
   > SetOverridePhysicsAsset!(
    execute
    Node:"Rigid Body Anim Node Reference Structure (by ref)"=@n2
    PhysicsAsset:Physics Asset Object Reference=Get Physics Asset() -> (Physics Asset:Physics Asset Object Reference)
   ) -> (then, ReturnValue:Rigid Body Anim Node Reference Structure)
   > Return Node()
  ]}
]
variables [
 BwdMovement {type=number}
 ChaosClothingInteractor {type=ChaosClothingInteractor, default=null}
 Cloth Controller Parameters {type=FortCharacterDynamicsParameters, default=/BRCosmetics/Accessories/FORT_Capes/Common/Base_Cape_v2/Fortnite_Base_Cape_v2_Parameters}
 Cloth Core Settings_0 {
  type=ClothCoreSettings
  default {
   collisionThickness=1.5
   dampingCoefficient=0.1
   frictionCoefficient=0.8
   localDampingCoefficient=0
   numberIterations=2
   numberSubsteps=1
   selfCollisionThickness=0
   animDriveDamping=<0 1>
   animDriveStiffness=<0 1>
   areaStiffness=<1 1>
   bendingStiffness=<1 1>
   edgeStiffness=<1 1>
   tetherStiffness=<1 1>
  }
 }
 Physics Asset {type=PhysicsAsset, default=/BRCosmetics/Accessories/FORT_Capes/Common/Base_Cape_v2/Fortnite_Base_Cape_v2_Dyn, class=~, bodies=~}
 resetClothLandingFromZilpine {type=boolean}
 SkydivingAlpha {type=Rotator}
 Wind Gust Local Settings {type=WindGustSettings, default {elapsedTime=0, finalGustStrengthLocal=0, gustDuration=0, gustStrength=0, timeBetweenGusts=0}}
 Wind Gust Local Settings Out {type=WindGustSettings, default {elapsedTime=0, finalGustStrengthLocal=0, gustDuration=0, gustStrength=0, timeBetweenGusts=0}}
]
